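# Behavior tree for the Vampire Boss NPC's katana attack, using its own KatanaAttack blackboard.
# A root Selector runs a focus-handling evaluator service over a Sequence that moves toward the evaluator's target actor, then runs Use Item on the katana action-object key.
# If that Sequence fails, the Selector falls back to a short Wait.
# No decorators gate either branch, so the attack relies on blackboard keys and task results.
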BehaviorTree /VampireBoss/NPCs/VampireBoss/BTInjection/BT_KatanaAttack_VampireBoss
blackboard=&/BB_KatanaAttack_VampireBoss
root {
 name=Selector
 class=BTComposite_Selector
 children [
  Sequence {
   class=BTComposite_Sequence
   children [
    Move To {
     class=BTTask_MoveTo
     children []
     decorators []
     properties {
      bIgnoreRestartSelf=false
      nodeName=Move To
      acceptableRadius {defaultValue=150, key=None}
      bAllowPartialPath {defaultValue=true, key=None}
      bAllowStrafe {defaultValue=false, key=None}
      bProjectGoalLocation {defaultValue=true, key=None}
      bReachTestIncludesAgentRadius {defaultValue=true, key=None}
      bReachTestIncludesGoalRadius {defaultValue=true, key=None}
      bRequireNavigableEndLocation {defaultValue=true, key=None}
      bStartFromPreviousPath {defaultValue=false, key=None}
      bTrackMovingGoal {defaultValue=true, key=None}
      blackboardKey {bNoneIsAllowedValue=false, selectedKeyName=AIEvaluator_Global_TargetActor}
      filterClass {baseClass=&NavigationQueryFilter, defaultValue=null, key=None}
      observedBlackboardValueTolerance {defaultValue=4.75, key=None}
     }
     services []
    }
    Use Item {
     class=FortAthenaBTTask_UseItem
     decorators []
     properties {
      actionObjectKeyName=AIEvaluator_KatanaAttackVamp_ActionObject
      executionStatusKeyName=AITask_KatanaAttackVamp_UseItemExecutionStatus
      nodeName=Use Item
      finishAfterItemUse {defaultValue=false}
      inputHoldDuration {key=None}
      maximumWaitTimeBetweenUses {defaultValue=1}
      resetActionObjectKey {defaultValue=true}
     }
    }
   ]
   decorators []
   properties {children [
     {childTask=&self:BTTask_MoveTo_0}
     {childTask=&self:FortAthenaBTTask_UseItem_0}
    ]}
  }
  Wait {class=BTTask_Wait, properties {waitTime {defaultValue=2}}}
 ]
 decorators []
 properties {
  children [
   {childComposite=&self:BTComposite_Sequence_0, childTask=null, decorators []}
   {childComposite=null, childTask=&self:BTTask_Wait_0, decorators []}
  ]
  services [&self:FortAthenaBTService_AIEvaluator_55]
 }
 services [Focus on target {
   class=FortAthenaBTService_AIEvaluator
   properties {interval=0.5, nodeName=Focus on target, aIEvaluatorClass {defaultValue=&FortAthenaAIBotEvaluator_HandleFocusing}}
  }]
}
