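# BT_LifeDrain is an injectable behavior tree that lets the Vampire Boss NPC use its Life Drain ability.
# A root Selector holds a single task that executes the gameplay ability tagged NPCLibrary.Abilities.LifeDrain.
# Two decorators gate the task, checking that the ability exists and can activate against SelfActor.
# The activation check aborts in both directions, so the tree reacts when the ability's availability changes.
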BehaviorTree /VampireBoss/NPCs/VampireBoss/BTInjection/BT_LifeDrain
blackboard=&/BB_LifeDrain
root {
 name=Selector
 class=BTComposite_Selector
 children [TriggerAbility {
   class=FortBTTask_ExecuteGameplayAbility
   decorators [
    Has Gameplay Ability {
     class=FortBTDecorator_GameplayAbility_HasGameplayAbility
     properties {gameplayAbilityTagContainer {defaultValue {gameplayTags [{tagName=NPCLibrary.Abilities.LifeDrain}]}}}
    }
    Can Activate Gameplay Ability {
     class=FortBTDecorator_GameplayAbility_CanActivate
     properties {
      flowAbortMode=Both
      gameplayAbilityTagContainer {defaultValue {gameplayTags [{tagName=NPCLibrary.Abilities.LifeDrain}]}}
      targetValue {key=SelfActor}
     }
    }
   ]
   properties {nodeName=TriggerAbility, gameplayAbilityTagContainer {defaultValue {gameplayTags [{tagName=NPCLibrary.Abilities.LifeDrain}]}}}
  }]
 decorators []
 properties {
  children [{
    childComposite=null
    childTask=&self:FortBTTask_ExecuteGameplayAbility_4
    decorators [
     &self:FortBTDecorator_GameplayAbility_HasGameplayAbility_1
     &self:FortBTDecorator_GameplayAbility_CanActivate_1
    ]
   }]
  services []
 }
 services []
}
