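; Behavior tree evaluator that gates the Vampire Boss bot's katana attack.
; On enter, it stores the katana-tagged slot item in the action object blackboard key and caches the blackboard and ability system components.
; Starting and ongoing checks share one condition: the bot pawn is not surface swimming, has line of sight on the threat per the blackboard, and its current weapon carries the katana tag.
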(event EventOnEnter
  (params (OwnerComp :type Behavior%20Tree%20Component%20Object%20Reference))
  (bind _returnvalue (Components|GetOwner OwnerComp))
  (CallFunction|SetActionItemKey OwnerComp)
  (Variables|Default|SetBlackboard%20Component (AI|GetBlackboard _returnvalue))
  (bind _asfort_athena_aibot_controller (Utilities|Casting|CastToFortAthenaAIBotController _returnvalue))
  (Variables|Default|SetAbility%20System%20Component (Ability|GetFortAbilitySystemComponent (Variables|/Script/FortniteAI.FortAthenaAIBotController|GetPlayerBotPawn _asfort_athena_aibot_controller))))

(fn EvaluateOngoingConditions
  (params (OwnerComp :type Behavior%20Tree%20Component%20Object%20Reference))
  (bind _canusekatana (CallFunction|KatanaConditions OwnerComp))
  (return _canusekatana))

(fn EvaluateStartingConditions
  (params (OwnerComp :type Behavior%20Tree%20Component%20Object%20Reference))
  (bind _canusekatana (CallFunction|KatanaConditions OwnerComp))
  (return _canusekatana))

(fn GetActionItem
  (params (OwnerComp :type Behavior%20Tree%20Component%20Object%20Reference))
  (bind _asfort_world_item (Utilities|Casting|CastToFortWorldItem (AI|Components|Blackboard|GetValueasObject (AI|GetBlackboard (Components|GetOwner OwnerComp)) (Variables|Default|GetAction%20Object%20Key)))
    (:then
      _)
    (:CastFailed
      _))
  (return _asfort_world_item))

(fn GetCurrentTarget
  (params (BlackboardComp :type Blackboard%20Component%20Object%20Reference))
  (bind _asactor (Utilities|Casting|CastToActor (AI|Components|Blackboard|GetValueasObject BlackboardComp (Variables|Default|GetAIEvaluator_Global_TargetActor)))
    (:then
      (return _asactor))
    (:CastFailed
      (return 0))))

(fn HasLOSOnTarget
  (params (BlackboardComp :type Blackboard%20Component%20Object%20Reference))
  (return (AI|Components|Blackboard|GetValueasBool BlackboardComp (Variables|Default|GetAIEvaluator_Global_HasLoSOnThreat))))

(fn KatanaConditions
  (params (OwnerComp :type Behavior%20Tree%20Component%20Object%20Reference))
  (bind _blackboard_component (Variables|Default|GetBlackboard%20Component))
  (Utilities|IsValid _blackboard_component
    (:Is%20Valid
      (bind _asfort_athena_aibot_controller (Utilities|Casting|CastToFortAthenaAIBotController (Components|GetOwner OwnerComp))
        (:then
          (Utilities|IsValid (Variables|/Script/FortniteAI.FortAthenaAIBotController|GetPlayerBotPawn _asfort_athena_aibot_controller)
            (:Is%20Valid
              (return (Math|Boolean|ANDBoolean (not (WaterBody|IsSurfaceSwimming (Variables|/Script/FortniteAI.FortAthenaAIBotController|GetPlayerBotPawn _asfort_athena_aibot_controller))) (CallFunction|HasLOSOnTarget _blackboard_component) (GameplayTags|HasTag (Tags|GetWeaponDataGameplayTags (Weapon|GetCurrentWeapon (Variables|/Script/FortniteAI.FortAthenaAIBotController|GetPlayerBotPawn _asfort_athena_aibot_controller))) (Variables|Default|GetTagKatana)))))
            (:Is%20Not%20Valid
              _)))
        (:CastFailed
          _))
      (return false))
    (:Is%20Not%20Valid
      _)))

(fn SetActionItemKey
  (params (OwnerComp :type Behavior%20Tree%20Component%20Object%20Reference))
  (bind _returnvalue (Components|GetOwner OwnerComp))
  (bind _asfort_athena_aibot_controller (Utilities|Casting|CastToFortAthenaAIBotController _returnvalue))
  (AI|Components|Blackboard|SetValueasObject (AI|GetBlackboard _returnvalue) (Variables|Default|GetAction%20Object%20Key) (Class|FortAthenaAIBotController|GetSlotItembyTag _asfort_athena_aibot_controller (Variables|Default|GetTagKatana))))
